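# Sound cue for Dopey impact sounds, using pickaxe player impact attenuation and a quieter-ducking impact sound class.
# A random node picks one of four equally weighted impact waves and feeds it through a modulator at the root, with boosted overall volume.
# Overridden concurrency caps simultaneous instances, stops the oldest, and scales volume down as copies stack.
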SoundCue /Figment_S10/Athena/Sounds/Events/Dopey/Dopey_Imp_Cue
root=Modulator_0
nodes [
 {id=Modulator_0, children [Random_0]}
 {id=Random_0, children [WavePlayer_1,WavePlayer_2,WavePlayer_3,WavePlayer_4], properties {weights [1,1,1,1]}}
 {id=WavePlayer_1, properties {soundWaveAssetPtr=&/Dopey_Imp_04}}
 {id=WavePlayer_2, properties {soundWaveAssetPtr=&/Dopey_Imp_01}}
 {id=WavePlayer_3, properties {soundWaveAssetPtr=&/Dopey_Imp_02}}
 {id=WavePlayer_4, properties {soundWaveAssetPtr=&/Dopey_Imp_03}}
]
properties {
 attenuationSettings=/Game/Sounds/Attenuation/Fort_Weapon_Pickaxe_Player_Impact_Att
 bOverrideConcurrency=true
 maxDistance=3512
 soundClassObject=/Game/Sounds/FortSoundClassesAndModes/Fort_ImpactSounds_DuckDefaultQuieter_PSM
 volumeMultiplier=2
 attenuationOverrides {falloffDistance=3000, attenuationShapeExtents {x=0}}
 concurrencyOverrides {maxCount=4, resolutionRule=StopOldest, volumeScale=0.8}
}
